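# Sound cue for the Dark Side Rift Gate activation stinger in PlotTwist.
# A single wave player feeds the stinger wave straight to the output, with a volume boost and the Fort_Music_NPC sound class.
# A concurrency override caps simultaneous plays at a few per owner.
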SoundCue /PlotTwist/Sounds/Stingers/RiftGate_Stinger_DarkSide_Activate_Cue
root=WavePlayer_0
nodes [{id=WavePlayer_0, properties {soundWaveAssetPtr=&/RiftGate_Stinger_DarkSide_Activate}}]
properties {
 bOverrideConcurrency=true
 maxDistance=2097152
 soundClassObject=&Fort_Music_NPC
 volumeMultiplier=1.619
 concurrencyOverrides {bLimitToOwner=true, maxCount=3}
 concurrencySet [null]
}
