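# Sound cue for the Dark Side Force rift gate music stinger in PlotTwist.
# A single wave player feeds the stinger wave straight to the output with boosted volume, routed through the NPC music sound class.
# A concurrency override allows one instance at a time and blocks new plays rather than cutting off the one already playing.
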SoundCue /PlotTwist/Sounds/Stingers/RiftGate_Stinger_DarkSide_Force_Cue
root=WavePlayer_0
nodes [{id=WavePlayer_0, properties {soundWaveAssetPtr=&/RiftGate_Stinger_DarkSide_Force}}]
properties {
 bOverrideConcurrency=true
 maxDistance=2097152
 soundClassObject=&Fort_Music_NPC
 volumeMultiplier=1.619
 concurrencyOverrides {maxCount=1, resolutionRule=PreventNew}
 concurrencySet [null]
}
