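# Sound cue for the Rift Gate light-side activation stinger in PlotTwist.
# A single wave player feeds the RiftGate_Stinger_LightSide_Activate wave straight to the output with a slight volume boost.
# It uses the Fort_Music_NPC sound class and overrides concurrency to allow a small number of instances per owner.
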SoundCue /PlotTwist/Sounds/Stingers/RiftGate_Stinger_LightSide_Activate_Cue
root=WavePlayer_1
nodes [{id=WavePlayer_1, properties {soundWaveAssetPtr=&/RiftGate_Stinger_LightSide_Activate}}]
properties {
 bOverrideConcurrency=true
 maxDistance=2097152
 soundClassObject=&Fort_Music_NPC
 volumeMultiplier=1.2381
 concurrencyOverrides {bLimitToOwner=true, maxCount=3}
 concurrencySet [null]
}
