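# Sound cue for the RiftGate light-side appear stinger in PlotTwist, routed to the POI music sound class.
# A single wave player feeds the output directly, with a slight volume boost and play-when-silent virtualization.
# Inline attenuation overrides give long-range falloff with distance-based low-pass filtering, reverb send and wide stereo spread.
# Inline concurrency overrides cap simultaneous instances per owner.
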SoundCue /PlotTwist/Sounds/Stingers/RiftGate_Stinger_LightSide_Appear_Cue
root=WavePlayer_0
nodes [{id=WavePlayer_0, properties {soundWaveAssetPtr=&/RiftGate_Stinger_LightSide_Appear}}]
properties {
 bOverrideAttenuation=true
 bOverrideConcurrency=true
 maxDistance=9000
 soundClassObject=/Game/Sounds/FortSoundClassesAndModes/Fort_Music_POI_PSM
 virtualizationMode=PlayWhenSilent
 volumeMultiplier=1.1429
 attenuationOverrides {
  bAttenuateWithLPF=true
  falloffDistance=8000
  lPFFrequencyAtMax=400
  lPFRadiusMin=2000
  reverbDistanceMax=6000
  reverbWetLevelMax=0.7
  stereoSpread=2000
  attenuationShapeExtents {x=1000}
 }
 concurrencyOverrides {bLimitToOwner=true, maxCount=3}
 concurrencySet [null]
}
