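# Sound cue for the RiftGate "LightSide Force" stinger in PlotTwist, routed to the Fort_Music_NPC sound class.
# A fixed short Delay feeds a single Wave Player that plays the stinger wave at boosted volume.
# A concurrency override allows only one instance and blocks new plays while it sounds, and a very large max distance is set.
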SoundCue /PlotTwist/Sounds/Stingers/RiftGate_Stinger_LightSide_Force_Cue
root=Delay_0
nodes [
 {id=Delay_0, children [WavePlayer_1], properties {delayMax=0.1, delayMin=0.1}}
 {id=WavePlayer_1, properties {soundWaveAssetPtr=&/RiftGate_Stinger_LightSide_Force}}
]
properties {
 bOverrideConcurrency=true
 maxDistance=2097152
 soundClassObject=&Fort_Music_NPC
 volumeMultiplier=1.3333
 concurrencyOverrides {maxCount=1, resolutionRule=PreventNew}
 concurrencySet [null]
}
